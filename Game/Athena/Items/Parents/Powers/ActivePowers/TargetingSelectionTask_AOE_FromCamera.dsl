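; Targeting System AOE selection task for active powers that orients its area query by the player's view.
; It overrides source offset and rotation, resolving the Fort player controller from the targeting context's source actor.
; FromCameraOffset's X and Y follow the control rotation's forward and right vectors, while Z stays world-up; source rotation is the control rotation.
; Both overrides fall back to the parent implementation without a valid actor or controller.
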(fn GetSourceOffset
  (params (TargetingHandle :type Targeting%20Request%20Handle%20Structure%20%28by%20ref%29))
  (bind _returnvalue (Targeting|GetTargetingSubsystem TargetingHandle))
  (bind _returnvalue_2 (TargetingSystem|TargetingTypes|GetTargetingSourceContext _returnvalue TargetingHandle))
  (bind _sourceactor (Utilities|Struct|BreakTargetingSourceContext _returnvalue_2))
  (bind _returnvalue_3 (Fortnite|GetFortPlayerControllerfromActor _sourceactor))
  (Utilities|IsValid _sourceactor
    (:Is%20Valid
      (Utilities|IsValid _returnvalue_3
        (:Is%20Valid
          (return (Utilities|Operators|Add (* (.x (Variables|TargetAOESelection|GetFromCameraOffset)) (Math|Vector|GetForwardVector (Pawn|GetControlRotation _returnvalue_3 :!class /Script/Engine.Controller))) (* (.y (Variables|TargetAOESelection|GetFromCameraOffset)) (Math|Vector|GetRightVector (Pawn|GetControlRotation _returnvalue_3 :!class /Script/Engine.Controller))) (Math|Vector|MakeVector 0.0 0.0 (.z (Variables|TargetAOESelection|GetFromCameraOffset))))))
        (:Is%20Not%20Valid
          _)))
    (:Is%20Not%20Valid
      _))
  (bind _returnvalue_4 (|CallParentFunction (Variables|GetTargetingHandle)))
  (return _returnvalue_4))

(fn GetSourceRotation
  (params (TargetingHandle :type Targeting%20Request%20Handle%20Structure%20%28by%20ref%29))
  (bind _returnvalue (Targeting|GetTargetingSubsystem TargetingHandle))
  (bind _returnvalue_2 (TargetingSystem|TargetingTypes|GetTargetingSourceContext _returnvalue TargetingHandle))
  (bind _sourceactor (Utilities|Struct|BreakTargetingSourceContext _returnvalue_2))
  (bind _returnvalue_3 (Fortnite|GetFortPlayerControllerfromActor _sourceactor))
  (Utilities|IsValid _sourceactor
    (:Is%20Valid
      (Utilities|IsValid _returnvalue_3
        (:Is%20Valid
          (return (Math|Conversions|ToQuaternion(Rotator) (Pawn|GetControlRotation _returnvalue_3 :!class /Script/Engine.Controller))))
        (:Is%20Not%20Valid
          _)))
    (:Is%20Not%20Valid
      _))
  (bind _returnvalue_4 (|CallParentFunction (Variables|GetTargetingHandle)))
  (return _returnvalue_4))
