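; Targeting System trace selection task for active powers that, per its name, sweeps a capsule from the camera sized to the source pawn's capsule.
; Radius and half-height overrides add padding values to the scaled capsule of the FortPlayerPawn taken from the request's source context.
; The rotation override returns the inverse of the pawn's eye-view rotation so the swept capsule stays upright in world space.
; Each override falls back to the parent implementation when its pawn check fails.
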(fn GetSourcePawn
  (params (TargetingHandle :type Targeting%20Request%20Handle%20Structure))
  (bind _returnvalue (Targeting|GetTargetingSubsystem TargetingHandle))
  (Utilities|IsValid _returnvalue
    (:Is%20Valid
      (bind _aspawn (Utilities|Casting|CastToPawn (Utilities|Struct|BreakTargetingSourceContext (TargetingSystem|TargetingTypes|GetTargetingSourceContext _returnvalue TargetingHandle)))
        (:then
          (return _aspawn))
        (:CastFailed
          _)))
    (:Is%20Not%20Valid
      _))
  (return 0))

(fn GetSweptTraceCapsuleHalfHeight
  (params (TargetingHandle :type Targeting%20Request%20Handle%20Structure%20%28by%20ref%29))
  (bind _asfort_player_pawn (Utilities|Casting|CastToFortPlayerPawn (CallFunction|GetSourcePawn TargetingHandle))
    (:then
      (return (+ (Components|Capsule|GetScaledCapsuleHalfHeight (Variables|/Script/Engine.Character|GetCapsuleComponent _asfort_player_pawn)) (Variables|TargetTraceSelection|GetHF_Padding_CapsuleHalfHeight))))
    (:CastFailed
      (bind _returnvalue (|CallParentFunction TargetingHandle))
      (return _returnvalue))))

(fn GetSweptTraceRadius
  (params (TargetingHandle :type Targeting%20Request%20Handle%20Structure%20%28by%20ref%29))
  (bind _asfort_player_pawn (Utilities|Casting|CastToFortPlayerPawn (CallFunction|GetSourcePawn TargetingHandle))
    (:then
      (return (+ (Components|Capsule|GetScaledCapsuleRadius (Variables|/Script/Engine.Character|GetCapsuleComponent _asfort_player_pawn)) (Variables|TargetTraceSelection|GetHF_Padding_CapsuleRadius))))
    (:CastFailed
      (bind _returnvalue (|CallParentFunction TargetingHandle))
      (return _returnvalue))))

(fn GetSweptTraceRotation
  (params (TargetingHandle :type Targeting%20Request%20Handle%20Structure%20%28by%20ref%29))
  (bind _pawn (CallFunction|GetSourcePawn TargetingHandle))
  (bind (_outlocation _outrotation) (Actor|GetActorEyesViewPoint _pawn))
  ; Inverse view direction into world space to capsule is straight up and down
  (Utilities|IsValid _pawn
    (:Is%20Valid
      (return (Math|Conversions|ToRotator(Quat) (Math|Quat|Inversed(Quat) (Math|Conversions|ToQuaternion(Rotator) _outrotation)))))
    (:Is%20Not%20Valid
      (bind _returnvalue (|CallParentFunction TargetingHandle))
      (return _returnvalue))))
